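; Weapon Blueprint for the DMR22 marksman rifle, handling the scope visuals and zoom sounds when aiming down sights.
; Targeting changes and local controllers' view-target changes (for spectating) drive a material effect that stashes the mesh materials, shows tinted scope glass and conditionally swaps in ADS materials.
; BeginPlay tints the glass by rarity; wrap application works around the glass material and reapplies active ADS materials.
; Locally controlled shots mark the spawned bullet as predicted.
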(event EventOnSetTargeting
  (params (bNewIsTargeting :type Boolean))
  (CallFunction|HandleLocalTargetingChange bNewIsTargeting))

(event EventBeginPlay
  (|CallParentFunction)
  (Utilities|FlowControl|Sequence
    (:then_0
      ;[1 1 1] Set scope's blue channel to the rarity color
      (bind _returnvalue (Rendering|Material|CreateDynamicMaterialInstance (Variables|Default|Getscope_glass) 0 (Rendering|Material|GetMaterial (Variables|Default|Getscope_glass) 0) "None" :!class /Script/Engine.PrimitiveComponent))
      (bind _returnvalue_2 (Fortnite|GetRarityData (Rarity|GetWeaponDataRarity)))
      (Rendering|Material|SetVectorParameterValue _returnvalue "Rarity_Color" (.Color2 (Utilities|Struct|BreakFortRarityItemData _returnvalue_2)) :!class /Script/Engine.MaterialInstanceDynamic))
    (:then_1
      ;[0 0.561 1] Handle spectating changes
      (if (not (Networking|IsDedicatedServer))
        (for _array_element (Fortnite|GetLocalFortPlayerControllers true true)
          (Utilities|IsValid _array_element
            (:Is%20Valid
              (EventDispatchers|AssignOnViewTargetChanged _array_element (delegate Custom|OnViewTargetChanged_Event)))))))))

(event Custom|OnViewTargetChanged_Event
  (params (PlayerController :type Fort%20Player%20Controller%20Object%20Reference) (OldViewTarget :type Actor%20Object%20Reference) (NewViewTarget :type Actor%20Object%20Reference))
  (bind _instigator (Variables|Actor|GetInstigator))
  ;[0 0.561 1] Handle spectating changes
  (Utilities|IsValid _instigator
    (:Is%20Valid
      (if (!= OldViewTarget NewViewTarget)
        (if (== OldViewTarget _instigator)
          (CallFunction|HandleLocalTargetingChange false)
          (elif (and (== NewViewTarget _instigator) (Camera|GetIsTargeting))
            (CallFunction|HandleLocalTargetingChange true)))))))

(event Targeting|EventForceScopeFX)

(event FortWeapon|EventOnSpawnProjectile
  (params (SpawnProjectile :type Fort%20Projectile%20Base%20Object%20Reference))
  (bind _instigator (Variables|Actor|GetInstigator)
    (:then
      (if (Pawn|IsLocallyControlled _instigator)
        (bind _asb_prj_bullet_dmr22 (Utilities|Casting|CastToB_Prj_Bullet_DMR22 SpawnProjectile))
        (Variables|/CobraDMR/DMR/B_Prj_Bullet_DMR22.B_Prj_Bullet_DMR22_C|SetIsPredicted true _asb_prj_bullet_dmr22)))
    (:else)))

(event EventOnWrapApplied
  (params (WrapDefinition :type Athena%20Item%20Wrap%20Definition%20Object%20Reference))
  (Utilities|IsValid WrapDefinition
    (:Is%20Valid
      ;[1 1 1] Temporary workaround to "skip" applying Midas touch to scope glass
      (Rendering|Material|SetMaterial (Variables|Visual|GetWeaponMesh) 2 "/CobraDMR/DMR/Materials/scope_material/MI_DMR_ScopeGlass_opaque.MI_DMR_ScopeGlass_opaque")
      ;[1 1 1] If wrap was applied while we were ADS, restore ADS materials and stash new wrap materials
      (if (Variables|Default|GetMaterialEffectActive)
        (CallFunction|TurnonScopeMaterialEffect)))))

(event EventEndPlay
  (params (EndPlayReason :type EEndPlayReason%20Enum :default "Destroyed"))
  ;[1 0.07939 0] Unbind events
  (|CallParentFunction EndPlayReason)
  (if (not (Networking|IsDedicatedServer))
    (for _array_element (Fortnite|GetLocalFortPlayerControllers true true)
      (Utilities|IsValid _array_element
        (:Is%20Valid
          (EventDispatchers|UnbindallEventsfromOnViewTargetChanged _array_element))))))

(fn HandleLocalTargetingChange
  (params (isTargeting :type Boolean))
  (Utilities|IsValid (Variables|Actor|GetInstigator)
    (:Is%20Valid
      (if isTargeting
        (bind _asfort_player_pawn (Utilities|Casting|CastToFortPlayerPawn (Variables|Actor|GetInstigator)))
        (if (Pawn|IsLocallyViewed _asfort_player_pawn)
          (CallFunction|TurnonScopeMaterialEffect)
          (Audio|PlaySound2D "/CobraDMR/DMR/Sounds/DMR22/DMR22_Scope_ZoomIn_Cue.DMR22_Scope_ZoomIn_Cue"))
        (else
          (bind _asfort_player_pawn_2 (Utilities|Casting|CastToFortPlayerPawn (Variables|Actor|GetInstigator)))
          (if (Pawn|IsLocallyViewed _asfort_player_pawn_2)
            (CallFunction|TurnOffScopeMaterialEffect)
            (Audio|PlaySound2D "/CobraDMR/DMR/Sounds/DMR22/DMR22_Scope_ZoomOut_Cue.DMR22_Scope_ZoomOut_Cue")
            ;[1 1 1] Withthout this, the pawn is invisible for spectators if they died while this pawn was scoped in
            (if (Utilities|Enum|Equal(Enum) (Networking|GetLocalRole (Variables|Actor|GetInstigator)) "ROLE_SimulatedProxy")
              (CallFunction|UnhideThirdPersonStuff))))))))

(fn ShouldHideWeaponMeshonADS
  (bind _returnvalue (Game|GetInstigator))
  (bind (_asfort_player_pawn _bsuccess) (Utilities|Casting|CastToFortPlayerPawn _returnvalue))
  (if _bsuccess
    (return (not (Targeting|IsDetachedfromCamera _asfort_player_pawn)))
    (else
      (return true))))

(fn TurnOffScopeMaterialEffect
  (bind _scope_glass (Variables|Default|Getscope_glass))
  ;[1 1 1] Restore old materials
  (Utilities|FlowControl|Sequence
    (:then_0
      (Rendering|Material|SetMaterial (Variables|Visual|GetWeaponMesh) 0 (Variables|Default|GetOldMaingunMaterial))
      (Rendering|Material|SetMaterial (Variables|Visual|GetWeaponMesh) 1 (Variables|Default|GetOldScopeMaterial))
      (Rendering|Material|SetMaterial (Variables|Visual|GetWeaponMesh) 2 (Variables|Default|GetOldScopeGlassMaterial))
      ;[1 1 1] Use this when wrap is disabled
      (Rendering|SetVisibility _scope_glass false false)
      (Rendering|Material|SetScalarParameterValueonMaterials _scope_glass "ScopeGlassWP" 0.0))
    (:then_1
      (Variables|Default|SetMaterialEffectActive false))))

(fn TurnOnScopeMaterialEffect
  (bind _scope_glass (Variables|Default|Getscope_glass))
  ;[1 1 1] Swap materials and store the previous material
  (Utilities|FlowControl|Sequence
    (:then_0
      (Variables|Default|SetOldMaingunMaterial (Rendering|Material|GetMaterial (Variables|Visual|GetWeaponMesh) 0))
      (Variables|Default|SetOldScopeMaterial (Rendering|Material|GetMaterial (Variables|Visual|GetWeaponMesh) 1))
      (Variables|Default|SetOldScopeGlassMaterial (Rendering|Material|GetMaterial (Variables|Visual|GetWeaponMesh) 2))
      ;[0.04 0.264 1] Use this when wrap is disabled
      (Rendering|SetVisibility _scope_glass true false)
      (Rendering|Material|SetScalarParameterValueonMaterials _scope_glass "ScopeGlassWP" 1.0))
    (:then_1
      (if (CallFunction|ShouldHideWeaponMeshOnADS)
        (Rendering|Material|SetMaterial (Variables|Visual|GetWeaponMesh) 0 "/CobraDMR/DMR/Materials/scope_material/MI_Designated_Marksman_Rifle_ADS_Inst.MI_Designated_Marksman_Rifle_ADS_Inst")
        (Rendering|Material|SetMaterial (Variables|Visual|GetWeaponMesh) 1 "/CobraDMR/DMR/Materials/scope_material/MI_Designated_Marksman_Rifle_Scope_ADS_Inst.MI_Designated_Marksman_Rifle_Scope_ADS_Inst")
        (Rendering|Material|SetMaterial (Variables|Visual|GetWeaponMesh) 2 "/CobraDMR/DMR/Materials/scope_material/MI_DMR_ScopeGlass_opaque_ADS.MI_DMR_ScopeGlass_opaque_ADS")))
    (:then_2
      (Variables|Default|SetMaterialEffectActive true))))

(fn ConstructionScript
  (|CallParentFunction))
